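# Lit material for the Deployable Storm gadget's bottle on skeletal meshes, producing an animated two-color storm-cloud glow alongside diffuse, normal and mask textures.
# A cloud mask is world-aligned-projected using pre-skinned position, so it follows the animated mesh, scrolled by time and offset along the camera vector for apparent depth.
# Contrast-adjusted clouds blend the colors, fade with inverse Fresnel, then sum a faint overall glow with one masked by the bottle's alpha texture.
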Material /Game/Gadgets/Assets/Deployable_Storm/Materials/M_DeployableStorm_Bottle
graph {
 properties {bUsedWithMeshDeformer=true, bUsedWithSkeletalMesh=true, pixelDepthOffsetMode=PDOM_Legacy, usedShadingModels=Default Lit}
 nodes [
  n1 = Add(
   A=VertexInterpolator(VS=PreSkinnedPosition()).PS
   B=Multiply(A=Time(), B=VectorParameter() {parameterName=Speed, defaultValue=<rgba 2 2 -5 0>}.RGB)
  )
  n2 = Multiply(A=CameraVectorWS(), B=ScalarParameter() {defaultValue=-3, parameterName=Offset})
  n3 = VertexNormalWS()
  n4 = ComponentMask(None=MaterialFunctionCall(
    TextureObject=TextureObject() {samplerType=SAMPLERTYPE_Masks, texture=/Game/Athena/SafeZone/Textures/StormCloudtubes_M}
    TextureSize=ScalarParameter() {defaultValue=18, parameterName=TextureScale}
    WorldPosition=Add(A=@n1, B=@n2)
    World Space Normal=Transform(None=@n3) {transformSourceType=TRANSFORMSOURCE_World, transformType=TRANSFORM_Local}
   ) {materialFunction=&WorldAlignedTexture}.XYZ Texture) {b=true}
  n5 = ScalarParameter() {defaultValue=3, parameterName=Cloud Contrast}
  n6 = Fresnel(ExponentIn=ScalarParameter() {defaultValue=1, parameterName=FresnelExponent})
  n7 = Multiply(
   A=LinearInterpolate(
    A=VectorParameter() {parameterName=Color1, defaultValue=<rgba 0.1081 0.0212 0.3698 1>}.RGB
    B=VectorParameter() {parameterName=Color2, defaultValue=<rgba 47.5069 11.9791 100 1>}.RGB
    Alpha=Power(Base=@n4, Exp=@n5)
   )
   B=Power(Base=OneMinus(None=@n6), Exp=ScalarParameter() {defaultValue=5, parameterName=FresnelPower})
  )
  Add(
   A=Multiply(A=@n7, B=ScalarParameter() {defaultValue=0.0025, parameterName=UnderPaintGlowAmount})
   B=Multiply(
    A=@n7
    B=TextureSample() {samplerType=SAMPLERTYPE_Masks, texture=/Game/Gadgets/Assets/Deployable_Storm/Textures/T_NEW_Deployable_Storm_ALPHA}.G
   )
  )
  BumpOffset()
  BumpOffset()
  TextureSample() {samplerType=SAMPLERTYPE_Normal, texture=/Game/Gadgets/Assets/Deployable_Storm/Textures/T_NEW_Deployable_Storm_N}
  TextureSample() {texture=/Game/Gadgets/Assets/Deployable_Storm/Textures/T_NEW_Deployable_Storm_D}
  TextureSample() {samplerType=SAMPLERTYPE_Masks, texture=/Game/Gadgets/Assets/Deployable_Storm/Textures/T_NEW_Deployable_Storm_S}
  TransformPosition(None=WorldPosition().XYZ) {transformSourceType=TRANSFORMPOSSOURCE_World}
 ]
}
